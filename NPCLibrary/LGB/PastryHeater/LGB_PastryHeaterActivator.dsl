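; Local gameplay behavior that turns on the passive healer of a PastryHeater NPC bot.
; On execute, with authority, it stores the activating Athena AI bot controller and finds the bot pawn's PassiveHealer component.
; It waits briefly so the pawn is ready before starting the healer, and also calls FinishExecute.
; On abort, the server uses the stored controller to stop that healer.
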(event EventBeginPlay)

(event Collision|EventActorBeginOverlap
  (params (OtherActor :type Actor%20Object%20Reference)))

(event EventTick
  (params (DeltaSeconds :type Float%20%28single-precision%29)))

(event LocalGameplayBehavior|EventOnExecute
  (params (Activator :type Actor%20Object%20Reference))
  (Networking|SwitchHasAuthority
    (:Authority
      (bind _asfort_athena_aibot_controller (Utilities|Casting|CastToFortAthenaAIBotController Activator)
        (:then
          (bind _output_get (Variables|Default|SetAIBotControllerActivator _asfort_athena_aibot_controller))
          (Utilities|IsValid (Variables|/Script/FortniteAI.FortAthenaAIBotController|GetPlayerBotPawn _output_get)
            (:Is%20Valid
              (Utilities|IsValid (Actor|GetComponentbyClass (Variables|/Script/FortniteAI.FortAthenaAIBotController|GetPlayerBotPawn _output_get) "/Script/FortniteGame.FortPawnComponent_PassiveHealer")
                (:Is%20Valid
                  (Utilities|FlowControl|Delay 2.0) ; Delay to be sur that Pawn is ready
                  (PassiveHealer|Start (Actor|GetComponentbyClass (Variables|/Script/FortniteAI.FortAthenaAIBotController|GetPlayerBotPawn _output_get) "/Script/FortniteGame.FortPawnComponent_PassiveHealer")))
                (:Is%20Not%20Valid
                  _)))
            (:Is%20Not%20Valid
              _))
          (LocalGameplayBehavior|FinishExecute))
        (:CastFailed
          _)))))

(event LocalGameplayBehavior|EventOnAbort
  (Networking|SwitchHasAuthority
    (:Authority
      (bind _aibotcontrolleractivator (Variables|Default|GetAIBotControllerActivator)
        (:then
          (Utilities|IsValid (Variables|/Script/FortniteAI.FortAthenaAIBotController|GetPlayerBotPawn _aibotcontrolleractivator)
            (:Is%20Valid
              (Utilities|IsValid (Actor|GetComponentbyClass (Variables|/Script/FortniteAI.FortAthenaAIBotController|GetPlayerBotPawn _aibotcontrolleractivator) "/Script/FortniteGame.FortPawnComponent_PassiveHealer")
                (:Is%20Valid
                  (PassiveHealer|Stop (Actor|GetComponentbyClass (Variables|/Script/FortniteAI.FortAthenaAIBotController|GetPlayerBotPawn _aibotcontrolleractivator) "/Script/FortniteGame.FortPawnComponent_PassiveHealer")))))))
        (:else)))))

(fn ConstructionScript)
